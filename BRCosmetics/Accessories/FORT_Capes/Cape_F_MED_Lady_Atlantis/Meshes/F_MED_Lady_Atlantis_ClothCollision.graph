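# Physics asset used as cloth collision for the Lady Atlantis female medium cape back accessory in Fortnite cosmetics.
# It holds a single body on the backpack_root bone made of two overlapping capsules, both rotated sideways, giving the cape's cloth a rounded volume to collide against.
# It has no constraints, so it serves only as a collision shape, not a simulated ragdoll.
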PhysicsAsset /BRCosmetics/Accessories/FORT_Capes/Cape_F_MED_Lady_Atlantis/Meshes/F_MED_Lady_Atlantis_ClothCollision
bodies {backpack_root [
  {shapeType=Capsule, length=8.7159, radius=6.1934, center=<0 -4.7579 -0.5268>, rotation=<rot 90 0 0>}
  {shapeType=Capsule, length=10.0262, radius=9.9803, shapeName=backpack_root_capsule, center=<-0.2331 -3.4192 7.0799>, rotation=<rot 90 0 0>}
 ]}
constraints []
